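# Unlit translucent material for Niagara ribbons in the Golden Outline vehicle boost cosmetic, drawing animated strips with sparkles.
# Strip texture channels pan at different speeds, with Dynamic Parameters offsetting and flipping some layers.
# Sine-pulsed sparkles fill areas outside the blue-channel mask, then join the strips, tinted by particle color through MF_FX_EmissiveAdjustments.
# A separate opacity chain multiplies panned strip alpha, particle alpha and a trail mask.
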Material /VehicleCosmetics/Boosts/Boost_GoldenOutline/Materials/M_Boost_ChineeseStrips
graph {
 properties {
  bAutomaticallySetUsageInEditor=false
  bUsedWithNiagaraRibbons=true
  blendMode=BLEND_Translucent
  shadingModel=MSM_Unlit
  usedShadingModels=Unlit
 }
 nodes [
  n1 = Panner(Coordinate=TextureCoordinate()) {speedX=-0.8}
  n2 = TextureSample(UVs=@n1) {texture=/VehicleCosmetics/Boosts/Boost_GoldenOutline/Textures/T_VCVFX_Boost_GoldenOutline_RGB}
  MaterialExpressionVertexColor_1 = ParticleColor()
  MaterialExpressionConstantClamp_8 = Saturate(None=Multiply(
    A=Multiply(A=@n2.A, B=@MaterialExpressionVertexColor_1.A)
    B=TextureSample() {texture=/VehicleCosmetics/Boosts/Boost_GoldenOutline/Textures/T_VCVFX_Boost_GoldenOutliner_TrailOpacity}.RGB
   ))
  Constant() {r=-0.075}
  Constant() {r=0.075}
  n3 = TextureCoordinate() {vTiling=-1}
  If(A=Constant() {r=1}, B=ScalarParameter() {defaultValue=1, parameterName=invert}, "A > B"=@n3, "A == B"=TextureCoordinate(), "A < B"=@n3)
  n4 = TextureSample(UVs=Panner(Coordinate=TextureCoordinate() {uTiling=6, vTiling=0.5}) {speedX=-1.2, speedY=-0.01}) {texture=/VehicleCosmetics/Boosts/Boost_GoldenOutline/Textures/T_VCVFX_Boost_GoldenOutline_Sparkles2_RGB}
  n5 = Time()
  MaterialExpressionConstantClamp_11 = Saturate(None=Sine(None=@n5) {period=2})
  n6 = Multiply(A=@n4.B, B=@MaterialExpressionConstantClamp_11)
  MaterialExpressionConstantClamp_12 = Saturate(None=Sine(None=@n5) {period=4})
  n7 = Multiply(A=@n4.R, B=@MaterialExpressionConstantClamp_12)
  MaterialExpressionConstantClamp_10 = Saturate(None=Add(A=Add(A=@n6, B=@n7), B=@n4.G))
  n8 = Constant() {r=3}
  n9 = DynamicParameter() {defaultValue=<rgba 0 0 1 1>, paramNames [offset,Flip,Param3,Param4]}
  n10 = TextureCoordinate()
  n11 = Add(
   A=AppendVector(A=@n9.offset, B=Constant())
   B=If(A=@n9.Flip, B=Constant() {r=1}, "A > B"=@n10, "A == B"=@n10, "A < B"=TextureCoordinate() {vTiling=-1})
  )
  n12 = TextureSample(UVs=Panner(Coordinate=@n11) {speedX=-0.5}) {texture=&T_VCVFX_Boost_GoldenOutline_RGB}
  MaterialExpressionConstantClamp_13 = Saturate(None=Multiply(A=Constant() {r=30}, B=@n12.B))
  MaterialExpressionConstantClamp_9 = Saturate(None=OneMinus(None=@MaterialExpressionConstantClamp_13))
  n13 = Multiply(A=Multiply(A=Multiply(A=@MaterialExpressionConstantClamp_10, B=@n8), B=@MaterialExpressionConstantClamp_9), B=Constant() {r=8})
  n14 = Panner(Coordinate=@n11) {speedX=-0.3}
  n15 = Add(
   A=Add(A=@n12.B, B=TextureSample(UVs=@n1) {texture=&T_VCVFX_Boost_GoldenOutline_RGB}.R)
   B=Multiply(
    A=@MaterialExpressionConstantClamp_9
    B=TextureSample(UVs=@n14) {texture=&T_VCVFX_Boost_GoldenOutline_RGB}.G
   )
  )
  MaterialFunctionCall(Emissive=Multiply(A=@MaterialExpressionVertexColor_1.RGB, B=Add(A=@n13, B=@n15))) {materialFunction=&MF_FX_EmissiveAdjustments}
 ]
}
